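# Niagara system for the StormYearn pickaxe's animation trails, drawing short-lived ribbons that follow bones or sockets on a skeletal mesh.
# Three ribbon emitters share anim-trail spawn, initialize and update modules: GlowTrail on one bone pair, SmallFlameTrail and SmokeTrail on another.
# Glow and flame fade through blue-cyan color curves, while dim smoke widens and animates dynamic material parameters.
# Time of day scales flame and smoke opacity and also tints the smoke.
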NiagaraSystem /BRCosmetics/Weapons/FORT_Melee/Pickaxe_StormYearn/FX/NS_Pickaxe_StormYearn_AnimTrails
dependencies {
 usedDataInterfaces [&NiagaraDataInterfaceColorCurve, &NiagaraDataInterfaceCurve, &NiagaraDataInterfaceSkeletalMesh]
 usedDynamicInputs []
 usedModules [
  &SystemState
  &self:SystemSpawnScript.SystemScriptSource.SystemScriptGraph.NiagaraNodeAssignment_0.SetVariables_#1
 ]
}
emitters {
 GlowTrail {
  emitterUpdateScript [
   &EmitterState()
   &FN_SpawnAnimTrail(AnimTrailWidthMode=::FromFirst, Spawn Interval=0)
  ]
  particleSpawnScript [
   &FN_InitializeAnimTrailBoneIndecies(First Bone or Socket Index=3, Second Bone or Socket Index=2)
   &FN_InitializeAnimTrail(Lifetime=0.06, SkeletalMesh=$System.Skeletal Mesh)
  ]
  particleUpdateScript [
   &SolveForcesAndVelocity(Clamp Velocity=false, Limit Acceleration=false)
   &FN_UpdateAnimTrail(SkeletalMesh=$System.Skeletal Mesh) {module=FN_UpdateAnimTrail001}
   &ScaleColor(
    Scale Alpha=&FloatFromCurve(
     FloatCurve=DataInterface() {propertyValues {Curve {keys [
         {arriveTangent=-2.003, interpMode=RCIM_Cubic, leaveTangent=-2.003, tangentMode=RCTM_User, time=0.5, value=0.95}
         {interpMode=RCIM_Cubic, tangentMode=RCTM_User, time=1, value=0}
        ]}}}
     Scale Curve=0.3
    )
    Scale RGB=&MakeVectorFromLinearColorRGB(Color=&ColorFromCurve(ColorCurve=DataInterface() {propertyValues {
        AlphaCurve {keys [
          {arriveTangent=-6.1994, interpMode=RCIM_Cubic, leaveTangent=-6.1994, tangentMode=RCTM_User, time=0.7, value=1}
          {arriveTangent=-0.0561, interpMode=RCIM_Cubic, tangentMode=RCTM_User, time=1, value=1}
         ]}
        BlueCurve {keys [
          {interpMode=RCIM_Cubic, tangentMode=RCTM_User, time=0, value=1.5}
          {interpMode=RCIM_Cubic, tangentMode=RCTM_User, time=0.0643, value=0.5}
          {interpMode=RCIM_Cubic, tangentMode=RCTM_User, time=0.3557, value=0.2}
         ]}
        GreenCurve {keys [
          {interpMode=RCIM_Cubic, tangentMode=RCTM_User, time=0, value=1.125}
          {interpMode=RCIM_Cubic, tangentMode=RCTM_User, time=0.0643, value=0.2917}
          {interpMode=RCIM_Cubic, tangentMode=RCTM_User, time=0.3557, value=0.08}
         ]}
        RedCurve {keys [
          {interpMode=RCIM_Cubic, tangentMode=RCTM_User, time=0, value=0}
          {interpMode=RCIM_Cubic, tangentMode=RCTM_User, time=0.0643, value=0}
          {interpMode=RCIM_Cubic, tangentMode=RCTM_User, time=0.3557, value=0.02}
         ]}
       }}))
    ScaleA=true
    ScaleRGB=true
   )
  ]
  rendererClasses [&NiagaraRibbonRendererProperties]
  renderers [{rendererClass=&NiagaraRibbonRendererProperties}]
 }
 SmallFlameTrail {
  emitterUpdateScript [
   &EmitterState()
   &FN_SpawnAnimTrail(Spawn Interval=0)
  ]
  particleSpawnScript [
   &FN_InitializeAnimTrailBoneIndecies(First Bone or Socket Index=1, Second Bone or Socket Index=0)
   &FN_InitializeAnimTrail(AnimTrailWidth=0.7, Lifetime=0.1, SkeletalMesh=$System.Skeletal Mesh)
  ]
  particleUpdateScript [
   &SolveForcesAndVelocity(Clamp Velocity=false, Limit Acceleration=false)
   &FN_UpdateAnimTrail(Locked Particle RibbonWidth=1, SkeletalMesh=$System.Skeletal Mesh) {module=FN_UpdateAnimTrail001}
   &ScaleColor(
    Scale Alpha=&Multiply_Float(
     A=&Multiply_Float(
      A=&FloatFromCurve(FloatCurve=DataInterface() {propertyValues {Curve {keys [
           {tangentMode=RCTM_User, time=0, value=1}
           {arriveTangent=-0.0405, interpMode=RCIM_Cubic, leaveTangent=-0.0405, tangentMode=RCTM_User, time=0.7009, value=1}
           {arriveTangent=-8.0089, interpMode=RCIM_Cubic, leaveTangent=-8.0089, tangentMode=RCTM_User, time=1, value=0}
          ]}}})
      B=&FNTimeOfDayScalar(Night Scale=0.95, Morning Scale=0.3, Evening Scale=0.3)
     )
     B=0.2
    )
    Scale RGB=&MakeVectorFromLinearColorRGB(Color=&ColorFromCurve(ColorCurve=DataInterface() {propertyValues {
        AlphaCurve {keys [
          {arriveTangent=-6.1994, interpMode=RCIM_Cubic, leaveTangent=-6.1994, tangentMode=RCTM_User, time=0.7, value=1}
          {arriveTangent=-0.0561, interpMode=RCIM_Cubic, tangentMode=RCTM_User, time=1, value=1}
         ]}
        BlueCurve {keys [
          {interpMode=RCIM_Cubic, tangentMode=RCTM_User, time=0, value=2}
          {interpMode=RCIM_Cubic, tangentMode=RCTM_User, time=0.0312, value=1}
          {interpMode=RCIM_Cubic, tangentMode=RCTM_User, time=0.1779, value=0.5}
          {interpMode=RCIM_Cubic, tangentMode=RCTM_User, time=0.7027, value=0.2}
         ]}
        GreenCurve {keys [
          {interpMode=RCIM_Cubic, tangentMode=RCTM_User, time=0, value=1.6667}
          {interpMode=RCIM_Cubic, tangentMode=RCTM_User, time=0.0312, value=0.75}
          {interpMode=RCIM_Cubic, tangentMode=RCTM_User, time=0.1779, value=0.25}
          {interpMode=RCIM_Cubic, tangentMode=RCTM_User, time=0.7027, value=0.08}
         ]}
        RedCurve {keys [
          {interpMode=RCIM_Cubic, tangentMode=RCTM_User, time=0, value=0}
          {interpMode=RCIM_Cubic, tangentMode=RCTM_User, time=0.0312, value=0}
          {interpMode=RCIM_Cubic, tangentMode=RCTM_User, time=0.1779, value=0}
          {interpMode=RCIM_Cubic, tangentMode=RCTM_User, time=0.7027, value=0.02}
         ]}
       }}))
    ScaleA=true
    ScaleRGB=true
   ) {module=ScaleColor002}
   &ScaleRibbonWidth(Ribbon Width Scale=&FloatFromCurve(FloatCurve=DataInterface() {propertyValues {Curve {keys [{interpMode=RCIM_Cubic, tangentMode=RCTM_User, time=0, value=1}, <at 1 0>]}}}))
  ]
  rendererClasses [&NiagaraRibbonRendererProperties]
  renderers [{rendererClass=&NiagaraRibbonRendererProperties}]
 }
 SmokeTrail {
  emitterUpdateScript [
   &EmitterState()
   &FN_SpawnAnimTrail(Spawn Interval=0)
  ]
  particleSpawnScript [
   &FN_InitializeAnimTrailBoneIndecies(First Bone or Socket Index=1, Second Bone or Socket Index=0)
   &FN_InitializeAnimTrail(AnimTrailWidth=4, Lifetime=0.24, SkeletalMesh=$System.Skeletal Mesh)
  ]
  particleUpdateScript [
   &SolveForcesAndVelocity(Clamp Velocity=false, Limit Acceleration=false)
   &FN_UpdateAnimTrail(Locked Particle RibbonWidth=1, SkeletalMesh=$System.Skeletal Mesh) {module=FN_UpdateAnimTrail001}
   &ScaleColor(
    Scale Alpha=&FloatFromCurve(
     FloatCurve=DataInterface() {propertyValues {Curve {keys [
         {arriveTangent=-2.287, interpMode=RCIM_Cubic, leaveTangent=-2.287, tangentMode=RCTM_User, time=0, value=1}
         {interpMode=RCIM_Cubic, tangentMode=RCTM_User, time=1, value=0}
        ]}}}
     Scale Curve=3
    )
    Scale RGB=&MakeVectorFromLinearColorRGB(Color=&ColorFromCurve(ColorCurve=DataInterface() {propertyValues {
        AlphaCurve {keys [<at 0 1 RCIM_Cubic>]}
        BlueCurve {keys [<at 0 0.05>,<at 0.1973 0.05>,<at 1 0 RCIM_Cubic>]}
        GreenCurve {keys [<at 0 0.0472>,<at 0.1973 0.03>,<at 1 0 RCIM_Cubic>]}
        RedCurve {keys [<at 0 0.025>,<at 0.1973 0.02>,<at 1 0 RCIM_Cubic>]}
       }}))
    ScaleA=true
    ScaleRGB=true
   ) {module=ScaleColor001}
   &ScaleColor(
    Scale Alpha=&FN_Select_Float_from_Time_Of_Day(Morning=0.7, Day=0.7, Evening=0.7, Night=0.5, Night Threshold=22)
    Scale RGB=&MakeVectorFromLinearColorRGB(Color=&ColorFromCurve(
      ColorCurve=DataInterface() {propertyValues {
        AlphaCurve {keys [<at 0 1>,<at 1 1>]}
        BlueCurve {keys [
          {
           arriveTangent=-0.1803
           arriveTangentWeight=0.2479
           interpMode=RCIM_Cubic
           leaveTangent=-0.1803
           leaveTangentWeight=0.2479
           tangentMode=RCTM_User
           tangentWeightMode=RCTWM_WeightedBoth
           time=0.0854
           value=1
          }
          {
           arriveTangentWeight=0.3333
           interpMode=RCIM_Cubic
           leaveTangentWeight=0.3333
           tangentMode=RCTM_User
           tangentWeightMode=RCTWM_WeightedBoth
           time=0.2905
           value=1
          }
          <at 0.8435 1>
          <at 0.9336 1>
         ]}
        GreenCurve {keys [
          {
           arriveTangent=-0.1803
           arriveTangentWeight=0.2479
           interpMode=RCIM_Cubic
           leaveTangent=-0.1803
           leaveTangentWeight=0.2479
           tangentMode=RCTM_User
           tangentWeightMode=RCTWM_WeightedBoth
           time=0.0854
           value=0.749
          }
          {
           arriveTangentWeight=0.3333
           interpMode=RCIM_Cubic
           leaveTangentWeight=0.3333
           tangentMode=RCTM_User
           tangentWeightMode=RCTWM_WeightedBoth
           time=0.2905
           value=1
          }
          <at 0.8435 1>
          <at 0.9336 0.749>
         ]}
        RedCurve {keys [
          {
           arriveTangent=-0.1803
           arriveTangentWeight=0.2479
           interpMode=RCIM_Cubic
           leaveTangent=-0.1803
           leaveTangentWeight=0.2479
           tangentMode=RCTM_User
           tangentWeightMode=RCTWM_WeightedBoth
           time=0.0854
           value=0
          }
          {
           arriveTangentWeight=0.3333
           interpMode=RCIM_Cubic
           leaveTangentWeight=0.3333
           tangentMode=RCTM_User
           tangentWeightMode=RCTWM_WeightedBoth
           time=0.2905
           value=1
          }
          <at 0.8435 1>
          <at 0.9336 0>
         ]}
       }}
      CurveIndex=Linked() {linkedVariable {name=NPC.FortniteNPC.RemappedTimeOfDay_0-1, type=NiagaraFloat}}
     ))
    ScaleA=true
    ScaleRGB=true
   )
   &ScaleRibbonWidth(Ribbon Width Scale=&FloatFromCurve(FloatCurve=DataInterface() {propertyValues {Curve {keys [
         {arriveTangent=1.0842, interpMode=RCIM_Cubic, leaveTangent=1.0842, tangentMode=RCTM_User, time=0.1, value=0.8}
         {arriveTangent=4.5044, interpMode=RCIM_Cubic, leaveTangent=4.5044, tangentMode=RCTM_User, time=0.9, value=3}
        ]}}})) {module=ScaleRibbonWidth001}
   &DynamicMaterialParameters(
    Index 0 Param 1=&FloatFromCurve(FloatCurve=DataInterface() {propertyValues {Curve {keys [
         {arriveTangent=0.2015, interpMode=RCIM_Cubic, leaveTangent=0.2015, tangentMode=RCTM_User, time=0, value=1}
         {arriveTangent=-1.5941, interpMode=RCIM_Cubic, leaveTangent=-1.5941, tangentMode=RCTM_User, time=0.3, value=0.5}
         {interpMode=RCIM_Cubic, tangentMode=RCTM_User, time=1, value=0}
        ]}}})
    Index 0 Param 4=&FloatFromCurve(
     FloatCurve=DataInterface() {propertyValues {Curve {keys [
         {interpMode=RCIM_Cubic, tangentMode=RCTM_User, time=0, value=1}
         {arriveTangent=-0.6219, interpMode=RCIM_Cubic, leaveTangent=-0.6219, tangentMode=RCTM_User, time=1, value=0}
        ]}}}
     Scale Curve=0.7
    )
    Module.Param0WriteEnabled=true
    Module.Param1WriteEnabled=true
    Module.Param2WriteEnabled=true
    Module.Param3WriteEnabled=true
   )
  ]
  rendererClasses [&NiagaraRibbonRendererProperties]
  renderers [{rendererClass=&NiagaraRibbonRendererProperties}]
 }
}
summary {
 systemName=NS_Pickaxe_StormYearn_AnimTrails
 emitters [
  {emitterName=SmallFlameTrail, rendererClasses [&NiagaraRibbonRendererProperties]}
  {emitterName=GlowTrail, rendererClasses [&NiagaraRibbonRendererProperties]}
  {emitterName=SmokeTrail, rendererClasses [&NiagaraRibbonRendererProperties]}
 ]
 userVariables [User.RibbonIDOffset {type=NiagaraInt32}]
}
